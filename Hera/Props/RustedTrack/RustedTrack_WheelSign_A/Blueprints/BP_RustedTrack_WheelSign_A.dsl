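; Blueprint for a rusted track wheel sign prop that emits recurring creak sounds as ambient audio.
; BeginPlay calls PlayRandomCreak, which plays the Random Creak One-Shots audio component from the start.
; The event then re-arms itself with a timer of random delay, so creaks repeat at irregular intervals of a few seconds.
; The construction script only calls its parent.
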(event EventBeginPlay
  (CallFunction|PlayRandomCreak))

(event Custom|PlayRandomCreak
  (Audio|Components|Audio|Play (Variables|Default|GetRandom%20Creak%20One-Shots) 0.0)
  (Utilities|Time|SetTimerbyEvent (delegate Custom|PlayRandomCreak) (Math|Random|RandomFloatinRange 2.5 4.0)))

(fn ConstructionScript
  (|CallParentFunction))
