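# MetaSound Source that plays a one-shot mono movement sound for the arm of the RustedTrack wheel sign prop.
# On play, a random array get picks one of five WheelSignArm movement waves and triggers the wave player.
# A random float, with no trigger wired to it, supplies a slight pitch shift for variation.
# The player's finish ends the source, which uses its own attenuation asset for spatial falloff.
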MetaSoundSource /Hera/Props/RustedTrack/RustedTrack_WheelSign_A/Sounds/MSS_WheelSignArm
inputs [
 In Array 1 {type=WaveAsset:Array}
 UE.Source.OnPlay {type=Trigger}
]
interfaces [UE.OutputFormat.Mono:1.0, UE.Source.OneShot:1.0, UE.Source:1.1]
outputs [
 "UE.OutputFormat.Mono.Audio:0" {type=Audio}
 UE.Source.OneShot.OnFinished {type=Trigger}
]
pages [Default {
  input_defaults {
   UE.Source.OnPlay=null
   In Array 1 [
    /Hera/Props/RustedTrack/RustedTrack_WheelSign_A/Sounds/Soundwaves/WheelSignArm_Movement_01
    /Hera/Props/RustedTrack/RustedTrack_WheelSign_A/Sounds/Soundwaves/WheelSignArm_Movement_02
    /Hera/Props/RustedTrack/RustedTrack_WheelSign_A/Sounds/Soundwaves/WheelSignArm_Movement_03
    /Hera/Props/RustedTrack/RustedTrack_WheelSign_A/Sounds/Soundwaves/WheelSignArm_Movement_04
    /Hera/Props/RustedTrack/RustedTrack_WheelSign_A/Sounds/Soundwaves/WheelSignArm_Movement_05
   ]
  }
  nodes [
   ArrayRandomGet_0 = Array.Random Get:WaveAsset:Array(In Array=^In Array 1, Next=^UE.Source.OnPlay)
   WavePlayer_0 = UE.Wave Player:Mono(Play=@ArrayRandomGet_0.On Next, Wave Asset=@ArrayRandomGet_0.Value, Pitch Shift=UE.RandomFloat() {also {defaults {Max=2, Min=-1}}}.Value) {also {version [1,1]}}
  ]
  links [{dst="@UE.Source.OneShot.OnFinished", src=WavePlayer_0.On Finished}, {dst="@UE.OutputFormat.Mono.Audio:0", src=WavePlayer_0.Out Mono}]
 }]
properties {
 assetImportData=&self:AssetImportData
 attenuationSettings=&/ATT_WheelSignArm
 maxDistance=2400
 sampleRate=48000
 volume=3
}
